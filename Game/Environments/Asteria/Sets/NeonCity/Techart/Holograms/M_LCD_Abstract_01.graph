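# Lit material for an abstract NeonCity LCD screen, showing animated vertical bars behind an LCD pixel mask.
# UV columns get random phases, seeded by actor position and stepped time, driving sine-animated bar lengths that periodically flip direction.
# A three-color gradient, distorted and masked by panning triangles, colors the bars, with a faint inverse fill.
# A glass mask varies metallic and roughness, while a noise normal adds subtle surface detail.
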Material /Game/Environments/Asteria/Sets/NeonCity/Techart/Holograms/M_LCD_Abstract_01
graph {
 properties {bUsedWithInstancedStaticMeshes=true, bUsedWithNanite=true, pixelDepthOffsetMode=PDOM_Legacy, usedShadingModels=Default Lit}
 nodes [
  n1 = Multiply(A=TextureCoordinate(), B=AppendVector(A=ScalarParameter() {defaultValue=10, parameterName=Lines_U}, B=Constant() {r=1}))
  n2 = Frac(None=ComponentMask(None=@n1) {r=true})
  n3 = Constant() {r=0.5}
  n4 = Constant() {r=0.4}
  n5 = Constant() {r=0.95}
  n6 = ComponentMask(None=TextureCoordinate()) {g=true}
  n7 = Time()
  n8 = Sign(None=Cosine(None=@n7) {period=20})
  n9 = Time()
  n10 = Ceil(None=MaterialFunctionCall!(Time=Subtract(A=@n9) {constB=0.75}, Divisor=Constant() {desc=match sine period, r=20}, Mod=Constant() {r=128}) {materialFunction=/Game/Effects/Fort_Effects/Materials/_Functions/MF_ModuloTime}.Unclamped)
  n11 = ActorPositionWS()
  n12 = Constant3Vector() {constant=<rgba 6.8214 2.1443 1.7251 0>}
  n13 = Truncate!(None=Abs(None=DotProduct(A=@n11, B=@n12))) {is=class MaterialExpressionTruncate}
  n14 = Add(
   A=MaterialFunctionCall!(
    PrimaryTextureCoordinates=@n1
    2ndOrderTextureCoordinates
    RandomSeed=Add(A=@n10, B=@n13)
    NoiseDensityBias
    Bypass Noise Influence Crosstalk
   ) {materialFunction=/Game/Effects/Fort_Effects/Materials/_Functions/MF_RandomUVSample}.NoiseWithRange
   B=@n7
  )
  n15 = SphereMask(
   A=Constant() {r=1}
   B=LinearInterpolate(A=@n6, B=OneMinus(None=@n6), Alpha=Saturate(None=@n8))
   Radius=ConstantBiasScale(None=Sine(None=@n14) {period=20}) {bias=0.5, scale=1.2}
   Hardness=ScalarParameter() {defaultValue=0.8, parameterName=Hardness}
  )
  n16 = VectorParameter() {parameterName=Top, defaultValue=<rgba 47.62 2.5174 50 0>}
  n17 = VectorParameter() {parameterName=Middle, defaultValue=<rgba 3.682 4.9566 10 0>}
  n18 = VectorParameter() {parameterName=Bottom, defaultValue=<rgba 0.1783 0.0726 0.1962 0>}
  n19 = TextureSample(UVs=Panner(Coordinate=TextureCoordinate() {vTiling=12}) {bFractionalPart=true, speedX=0.0027, speedY=-0.025}) {samplerType=SAMPLERTYPE_LinearColor, texture=/Game/TechArt/Seasonal/LackLuster/Textures/T_Lackluster_Triangles}
  n20 = Add(A=ComponentMask(None=TextureCoordinate()) {g=true}, B=Multiply(A=Subtract(A=@n19.R) {constB=0.5}) {constB=0.1})
  n21 = Multiply(
   A=@n15
   B=Multiply(
    A=MaterialFunctionCall(A=@n16.RGB, B=@n17.RGB, C=@n18.RGB, Alpha=@n20) {materialFunction=&3ColorBlend}.Result
    B=@n19.G
   )
  )
  n22 = Multiply(A=OneMinus(None=@n15), B=Multiply(A=@n19.G, B=ScalarParameter() {defaultValue=0.01, parameterName=EdgeInvertedScale}))
  n23 = ScalarParameter() {defaultValue=200, parameterName=LCDTileU}
  n24 = ScalarParameter() {defaultValue=1000, parameterName=LCDTileV}
  n25 = TextureSample(UVs=Multiply(A=TextureCoordinate(), B=AppendVector(A=@n23, B=@n24))) {samplerType=SAMPLERTYPE_Masks, texture=/Game/Environments/Asteria/Sets/NeonCity/Techart/Textures/T_LCD_Panels_LCDSoft}
  Multiply(A=Multiply(A=SphereMask(A=@n2, B=@n3, Radius=@n4, Hardness=@n5), B=Add(A=@n21, B=@n22)), B=@n25.R)
  Multiply(A=VectorParameter() {parameterName=BaseColor, defaultValue=<rgba 0.0258 0.0412 0.0955 0>}.RGB, B=@n25.R)
  n26 = TextureSample() {samplerType=SAMPLERTYPE_Normal, texture=/Game/Characters/Player/Female/Medium/Bodies/F_MED_Bistro_Astronaut/FX/T_GaussianNoise_Normal}
  NamedRerouteDeclaration(None=Multiply(A=@n26.RGB, B=Constant3Vector() {constant=<rgba 0.015 0.015 1 0>})) {name=Normals, nodeColor=<rgba 1 0 0.6797 1>}
  n27 = TextureSample(UVs=TextureCoordinate() {uTiling=2, vTiling=2}) {samplerType=SAMPLERTYPE_Masks, texture=/Game/Environments/Asteria/Sets/NeonCity/Techart/Textures/T_GlassA_Masks}
  NamedRerouteDeclaration(None=LinearInterpolate(
    A=ScalarParameter() {defaultValue=-0.2, parameterName=MetallicLow}
    B=ScalarParameter() {defaultValue=8, parameterName=MetallicHigh}
    Alpha=@n27.R
   )) {name=Metallic, nodeColor=<rgba 1 0.2812 0 1>}
  NamedRerouteDeclaration(None=LinearInterpolate(
    A=ScalarParameter() {defaultValue=-0.1, parameterName=RoughnessLow}
    B=ScalarParameter() {defaultValue=3, parameterName=RoughnessHigh}
    Alpha=@n27.R
   )) {name=Roughness, nodeColor=<rgba 1 0.0234 0 1>}
  NamedRerouteUsage()
  NamedRerouteUsage()
  NamedRerouteUsage()
 ]
}
