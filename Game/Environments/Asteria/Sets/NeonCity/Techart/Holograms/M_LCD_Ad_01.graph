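# Default Lit material for NeonCity ad screens, rendering a texture parameter as an emissive LCD panel.
# Quantized UVs pixelate the image, with a periodic sweep adding glitch offsets.
# Image brightness, camera-distance boost, panel variation and random bulb outages size sphere masks over an RGB subpixel texture.
# A high-mip-bias blur adds glow, a ray tracing switch applies a Lumen tint, and glass masks and noise normals supply surface detail.
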Material /Game/Environments/Asteria/Sets/NeonCity/Techart/Holograms/M_LCD_Ad_01
graph {
 properties {pixelDepthOffsetMode=PDOM_Legacy, usedShadingModels=Default Lit}
 nodes [
  n1 = TextureSample(UVs=NamedRerouteUsage()) {samplerType=SAMPLERTYPE_Masks, texture=/Game/Environments/Asteria/Sets/NeonCity/Techart/Textures/T_LCD_Panels_LCDSoft}
  n2 = Constant() {r=1}
  n3 = MaterialFunctionCall!(
   PrimaryTextureCoordinates=NamedRerouteUsage()
   2ndOrderTextureCoordinates
   RandomSeed=ScalarParameter() {defaultValue=600.1541, parameterName=BulbOutageRandomSeed}
   NoiseDensityBias
   Bypass Noise Influence Crosstalk
  ) {materialFunction=/Game/Effects/Fort_Effects/Materials/_Functions/MF_RandomUVSample}
  n4 = ScalarParameter() {parameterName=BulbOutageLow}
  n5 = Constant() {r=1}
  n6 = TextureSampleParameter2D(UVs=NamedRerouteUsage()) {parameterName=Texture, texture=/Game/Environments/Asteria/Sets/NeonCity/Techart/Screen/Textures/Logos/TeddyBear}
  n7 = ScalarParameter() {defaultValue=100, parameterName=BulbStartFade}
  n8 = ScalarParameter() {defaultValue=1000, parameterName=BulbEndFade}
  n9 = NamedRerouteDeclaration(None=Distance(A=CameraPositionWS(), B=ActorPositionWS())) {name=DistanceFromPixelToCamera, nodeColor=<rgba 0.9688 1 0 1>}
  n10 = Multiply(
   A=Saturate(None=InverseLinearInterpolate(A=@n7, B=@n8, Value=@n9) {constA=400, constB=800})
   B=ScalarParameter() {defaultValue=2, parameterName=test}
  )
  n11 = Clamp(None=Add(A=@n6.RGB, B=Multiply(A=@n6.RGB, B=@n10) {constB=3})) {clampMode=CMODE_ClampMax, maxDefault=0.99}
  n12 = Power(Base=TextureSample(UVs=NamedRerouteUsage()) {samplerType=SAMPLERTYPE_Masks, texture=/Game/Environments/Asteria/Sets/NeonCity/Techart/Textures/T_LCD_Panels_output}.R) {constExponent=50}
  n13 = Multiply(A=@n11, B=Multiply(A=Subtract(A=@n12) {constB=-0.5}, B=ScalarParameter() {defaultValue=0.02, parameterName=RandomPanelStrength}))
  n14 = MaterialFunctionCall(Float3=Multiply(A=If(A=@n3.NoiseWithRange, "A > B"=@n4, "A < B"=@n5) {constB=0.998}, B=Add(A=@n11, B=@n13))) {materialFunction=&BreakOutFloat3Components}
  n15 = ScalarParameter() {parameterName=Hardness}
  n16 = SphereMask(A=@n1.R, B=@n2, Radius=@n14.R, Hardness=@n15) {hardnessPercent=0.8}
  n17 = SphereMask(A=@n1.G, B=@n2, Radius=@n14.G, Hardness=@n15) {hardnessPercent=0.8}
  n18 = Add(A=@n16, B=@n17)
  n19 = SphereMask(A=@n1.B, B=@n2, Radius=@n14.B, Hardness=@n15) {hardnessPercent=0.8}
  MaterialExpressionAppend3Vector_1 = MaterialXAppend3Vector!(A=@n16, B=@n17, C=@n19)
  n20 = Multiply(
   A=LinearInterpolate(A=Add(A=@n18, B=@n19), B=@MaterialExpressionAppend3Vector_1, Alpha=ScalarParameter() {defaultValue=0.95, parameterName=RGBLerp})
   B=VectorParameter() {parameterName=EmissiveColor, defaultValue=<rgba 46.8055 80 50.4608 0>}.RGB
  )
  n21 = VectorParameter() {parameterName=Lumen, defaultValue=<rgba 50 50 50 0>}
  n22 = TextureSampleParameter2D!(UVs, Bias=Constant() {r=7}, Apply View MipBias) {mipValueMode=TMVM_MipBias, parameterName=Texture, texture=&TeddyBear}
  Add(
   A=RayTracingQualitySwitch(Normal=@n20, RayTraced=Multiply(A=@n21.RGB, B=@n20))
   B=Multiply(A=@n22.RGB, B=ScalarParameter() {defaultValue=0.15, parameterName=BlurScale})
  )
  Max(A=VectorParameter() {parameterName=MinColorVal, defaultValue=<rgba 0.0479 0.0955 0.0843 0>}.RGB, B=@n6.RGB)
  Multiply(A=Multiply(A=VectorParameter() {parameterName=BaseColor, defaultValue=<rgba 0.0255 0.026 0.0255 0>}.RGB, B=@n1.A), B=Round(None=@n1.RGB))
  n23 = ScalarParameter() {defaultValue=400, parameterName=UVScale}
  n24 = NamedRerouteDeclaration(None=Multiply(A=TextureCoordinate(), B=@n23)) {name=ScaledUVs, nodeColor=<rgba 1 0 0.0938 1>}
  n25 = Ceil(None=@n24)
  n26 = ComponentMask(None=@n24) {r=true}
  n27 = Constant() {r=8}
  n28 = AppendVector(
   A=Constant()
   B=Subtract(A=MaterialFunctionCall(PrimaryTextureCoordinates=@n26, RandomSeed=@n27) {materialFunction=&MF_RandomUVSample}.NoiseWithRange) {constB=0.5}
  )
  n29 = Time()
  n30 = Abs(None=Add(A=Sine(None=@n29) {period=1.332}, B=Sine(None=@n29) {period=1.1548}))
  n31 = Power(Base=Saturate(None=Divide(A=@n30) {constB=1.7})) {constExponent=30}
  n32 = ComponentMask(None=TextureCoordinate()) {r=true}
  n33 = MaterialFunctionCall!(Time, Divisor=ScalarParameter() {defaultValue=2, parameterName=SweepSpeed}, Mod) {materialFunction=/Game/Effects/Fort_Effects/Materials/_Functions/MF_ModuloTime}
  n34 = Divide(A=Round(None=Multiply(A=@n33.Clamped 0-1) {constB=10})) {constB=10}
  n35 = ScalarParameter() {defaultValue=0.15, parameterName=SweepRadius}
  n36 = Constant() {r=0.8}
  n37 = Multiply(
   A=ScalarParameter() {defaultValue=0.015, parameterName=DistortionScale}
   B=Multiply(A=Round(None=@n31), B=SphereMask(A=@n32, B=@n34, Radius=@n35, Hardness=@n36))
  )
  NamedRerouteDeclaration(None=Add(A=Divide(A=@n25, B=@n23), B=Multiply(A=@n28, B=@n37))) {name=QuantizedUVs, nodeColor=<rgba 1 0.8672 0 1>}
  n38 = TextureSample(UVs=TextureCoordinate() {uTiling=2, vTiling=2}) {samplerType=SAMPLERTYPE_Masks, texture=/Game/Environments/Asteria/Sets/NeonCity/Techart/Textures/T_GlassA_Masks}
  NamedRerouteDeclaration(None=LinearInterpolate(
    A=ScalarParameter() {defaultValue=-0.2, parameterName=MetallicLow}
    B=ScalarParameter() {defaultValue=8, parameterName=MetallicHigh}
    Alpha=@n38.R
   )) {name=Metallic, nodeColor=<rgba 1 0.2812 0 1>}
  NamedRerouteDeclaration(None=LinearInterpolate(
    A=ScalarParameter() {defaultValue=-0.1, parameterName=RoughnessLow}
    B=ScalarParameter() {defaultValue=3, parameterName=RoughnessHigh}
    Alpha=@n38.R
   )) {name=Roughness, nodeColor=<rgba 1 0.0234 0 1>}
  n39 = TextureSample() {samplerType=SAMPLERTYPE_Normal, texture=/Game/Characters/Player/Female/Medium/Bodies/F_MED_Bistro_Astronaut/FX/T_GaussianNoise_Normal}
  NamedRerouteDeclaration(None=Multiply(A=@n39.RGB, B=Constant3Vector() {constant=<rgba 0.015 0.015 1 0>})) {name=Normals, nodeColor=<rgba 1 0 0.6797 1>}
  NamedRerouteUsage()
  NamedRerouteUsage()
  NamedRerouteUsage()
 ]
}
