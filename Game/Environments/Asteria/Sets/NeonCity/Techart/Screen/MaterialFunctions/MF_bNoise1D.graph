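# Material function that generates cheap procedural 1D noise from time multiplied by speed, stored with the NeonCity screen tech-art.
# The floored value hashes into a per-cell coefficient, which shapes a polynomial bump over the fractional part.
# A sign term based on the halved input alternates bump direction between cells.
# It outputs the cell index and the signed noise, labeled "Random Walk 0-1" although the bias-scale remap node feeds no output.
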MaterialFunction /Game/Environments/Asteria/Sets/NeonCity/Techart/Screen/MaterialFunctions/MF_bNoise1D
graph {nodes [
  n1 = Multiply(
   A=FunctionInput(Preview=Time()) {bUsePreviewValueAsDefault=true, inputName=time, inputType=FunctionInput_Scalar}
   B=FunctionInput() {inputName=speed, inputType=FunctionInput_Scalar, previewValue=<1 0 0 1>}
  )
  n2 = Floor(None=@n1) {desc=i}
  n3 = Multiply(A=@n2) {constB=0.1731}
  n4 = Subtract(A=Multiply(B=Frac(None=@n3) {desc=k}) {constA=16, desc=16*k}) {constB=4}
  n5 = Frac(None=@n1) {desc=f}
  n6 = Multiply(A=@n5, B=Subtract(A=@n5))
  n7 = Subtract(A=Frac(None=Divide(A=@n1))) {constB=0.5}
  n8 = Multiply(A=Subtract(A=Multiply(A=@n4, B=@n6)), B=Multiply(A=@n6, B=Sign(None=@n7)))
  ConstantBiasScale(None=@n8)
  FunctionOutput(None=@n8) {outputName=Random Walk 0-1}
  FunctionOutput(None=@n2) {outputName=index}
 ]}
