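# Master-style material for a Neon City LCD ad screen, showing a logo over a two-color background with emissive pixels.
# MF_AdaptiveCellsUVs quantizes UVs, which time-pulsed, randomly moving glitch bars distort.
# The composite gets random panel variation and bulb outages, then an LCD mask and emissive boost, with a separate ray-traced Lumen multiplier.
# Masks drive base, metallic and roughness; blurred-glow and animated-background branches are left unconnected.
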Material /Game/Environments/Asteria/Sets/NeonCity/Techart/Holograms/M_LCD_Ad_Source
graph {
 properties {bUsedWithInstancedStaticMeshes=true, bUsedWithNanite=true, pixelDepthOffsetMode=PDOM_Legacy, usedShadingModels=Default Lit}
 nodes [
  n1 = TextureSample(UVs=NamedRerouteUsage()) {samplerType=SAMPLERTYPE_Masks, texture=/Game/Environments/Asteria/Sets/NeonCity/Techart/Textures/T_LCD_Panels_LCDSoft}
  n2 = If(
   A=MaterialFunctionCall!(
    PrimaryTextureCoordinates=NamedRerouteUsage()
    2ndOrderTextureCoordinates
    RandomSeed=ScalarParameter() {defaultValue=600.1541, parameterName=BulbOutageRandomSeed}
    NoiseDensityBias
    Bypass Noise Influence Crosstalk
   ) {materialFunction=/Game/Effects/Fort_Effects/Materials/_Functions/MF_RandomUVSample}.NoiseWithRange
   "A > B"=ScalarParameter() {parameterName=BulbOutageLow}
   "A < B"=Constant() {r=1}
  ) {constB=0.998}
  n3 = TextureSampleParameter2D(UVs=Multiply(A=NamedRerouteUsage(), B=ScalarParameter() {defaultValue=1, parameterName="Background Tiling "})) {
   parameterName=BackGroundTexture
   samplerType=SAMPLERTYPE_Masks
   texture=/Game/Environments/Asteria/Sets/NeonCity/Techart/Screen/Textures/panelRGB
  }
  n4 = LinearInterpolate(
   A=VectorParameter() {parameterName=ColorBG, defaultValue=<rgba 0 0.4913 1 1>}.RGB
   B=VectorParameter() {parameterName=ColorAccent, defaultValue=<rgba 0.6189 0.803 0.974 1>}.RGB
   Alpha=ChannelMaskParameter(None=@n3.RGB) {parameterName=BG Mask Channel}.R
  )
  n5 = TextureSampleParameter2D(UVs=NamedRerouteUsage()) {parameterName=Primary Logo, texture=/Game/Environments/Asteria/Sets/NeonCity/Techart/Screen/Textures/Logos/TeddyBear}
  n6 = Clamp(None=LinearInterpolate(A=@n4, B=@n5.RGB, Alpha=@n5.RGB)) {clampMode=CMODE_ClampMax, maxDefault=0.99}
  n7 = NamedRerouteUsage()
  n8 = Subtract(A=Power(Base=TextureSample(UVs=@n7) {samplerType=SAMPLERTYPE_Masks, texture=/Game/Environments/Asteria/Sets/NeonCity/Techart/Textures/T_LCD_Panels_output}.R) {constExponent=50}) {constB=-0.5}
  n9 = ScalarParameter() {defaultValue=0.02, parameterName=RandomPanelStrength}
  n10 = Add(A=@n6, B=Multiply(A=@n6, B=Multiply(A=@n8, B=@n9)))
  n11 = Multiply(A=@n1.RGB, B=Multiply(A=Multiply(A=@n2, B=@n10), B=VectorParameter() {parameterName=EmissiveColor, defaultValue=<rgba 40 40 40 0>}.RGB))
  n12 = RayTracingQualitySwitch(Normal=@n11, RayTraced=Multiply(A=VectorParameter() {parameterName=Lumen, defaultValue=<rgba 8 8 8 0>}.RGB, B=@n11))
  n13 = TextureSampleParameter2D!(UVs=NamedRerouteUsage(), Bias=ScalarParameter() {defaultValue=6, parameterName=mipBiasGlow}, Apply View MipBias) {mipValueMode=TMVM_MipBias, parameterName=Texture, texture=&TeddyBear}
  Add(A=@n12, B=Multiply(A=Multiply(A=@n13.RGB, B=@n13.RGB), B=ScalarParameter() {defaultValue=0.5, parameterName=BlurScale}))
  n14 = Time()
  n15 = Cosine(None=Add(A=Multiply(A=@n3.R) {constB=4}, B=Multiply(A=@n14) {constB=0.1}))
  n16 = Floor(None=Multiply(A=ConstantBiasScale(None=@n15)) {constB=5})
  Add(
   A=Multiply(A=@n4, B=Divide(A=@n16) {constB=5})
   B=Multiply(A=VectorParameter() {parameterName=CoreLines, defaultValue=<rgba 0 0.2344 0.1496 0>}.RGB, B=@n3.B)
  )
  ComponentMask() {g=true, r=true}
  Constant()
  LinearInterpolate(A=LinearInterpolate(A=@n4, Alpha=@n3.G), Alpha=@n3.B)
  Multiply(A=VectorParameter() {parameterName=Panelling, defaultValue=<rgba 0 0.5234 0.776 0>}.RGB, B=@n3.R)
  n17 = MaterialFunctionCall!(
   bump
   offsetToEdge=ScalarParameter() {parameterName=EdgeOffset}
   uvscale=ScalarParameter() {defaultValue=1, parameterName=uvscale}
   PixelAspect
   cellSize=ScalarParameter() {defaultValue=0.0032, parameterName=CellSize}
  ) {materialFunction=/Game/Environments/Asteria/Sets/NeonCity/Techart/Screen/MaterialFunctions/MF_AdaptiveCellsUVs}
  n18 = MaterialFunctionCall(
   PrimaryTextureCoordinates=ComponentMask(None=NamedRerouteDeclaration(None=@n17.tilingUVs) {name=ScaledUVs, nodeColor=<rgba 1 0 0.0938 1>}) {r=true}
   RandomSeed=Constant() {r=8}
  ) {materialFunction=&MF_RandomUVSample}
  n19 = Time()
  n20 = Sine(None=@n19) {period=1.332}
  n21 = Sine(None=@n19) {period=1.1548}
  n22 = Divide(A=Abs(None=Add(A=@n20, B=@n21))) {constB=1.7}
  n23 = Round(None=Power(Base=Saturate(None=@n22)) {constExponent=3})
  n24 = ComponentMask(None=NamedRerouteUsage()) {r=true}
  n25 = Multiply(A=MaterialFunctionCall!(speed=ScalarParameter() {defaultValue=0.04, parameterName=randSpeed}, time) {materialFunction=/Game/Environments/Asteria/Sets/NeonCity/Techart/Screen/MaterialFunctions/MF_bNoise1D}.Random Walk 0-1) {constB=128}
  n26 = Saturate(None=MaterialFunctionCall!(In=Floor(None=@n25)) {materialFunction=/Game/Environments/Asteria/Sets/NeonCity/Techart/Screen/MaterialFunctions/MF_Rand1Dto1D1}.Random1D)
  n27 = Sign(None=Subtract(A=@n24, B=@n26))
  n28 = Add(A=@n24, B=ScalarParameter() {defaultValue=0.2, parameterName=GlitchBarThickness}) {constB=0.5}
  n29 = Sign(None=Multiply(A=Subtract(A=@n28, B=@n26)) {constB=-1})
  n30 = OneMinus(None=Saturate(None=Max(A=@n27, B=@n29)))
  n31 = Multiply(
   A=AppendVector(A=Constant(), B=Subtract(A=@n18.NoiseWithRange) {constB=0.5})
   B=Multiply(A=ScalarParameter() {defaultValue=0.01, parameterName=DistortionScale}, B=Multiply(A=@n23, B=@n30))
  )
  NamedRerouteDeclaration(None=Add(A=@n17.clamped_uvs, B=@n31)) {name=QuantizedUVs, nodeColor=<rgba 1 0.8672 0 1>}
  NamedRerouteDeclaration(None=@n17.UnPixelatedUVs) {name=clampedunpixel, nodeColor=<rgba 1 0.1641 0 1>}
  n32 = TextureSample(UVs=TextureCoordinate() {uTiling=2, vTiling=2}) {samplerType=SAMPLERTYPE_Masks, texture=/Game/Environments/Asteria/Sets/NeonCity/Techart/Textures/T_GlassA_Masks}
  NamedRerouteDeclaration(None=LinearInterpolate(A=ScalarParameter() {parameterName=MetallicLow}, B=ScalarParameter() {defaultValue=8, parameterName=MetallicHigh}, Alpha=@n32.R)) {name=Metallic, nodeColor=<rgba 1 0.2812 0 1>}
  NamedRerouteDeclaration(None=LinearInterpolate(A=ScalarParameter() {parameterName=RoughnessLow}, B=ScalarParameter() {defaultValue=3, parameterName=RoughnessHigh}, Alpha=@n32.R)) {name=Roughness, nodeColor=<rgba 1 0.0234 0 1>}
  n33 = TextureSample() {samplerType=SAMPLERTYPE_Normal, texture=/Game/Characters/Player/Female/Medium/Bodies/F_MED_Bistro_Astronaut/FX/T_GaussianNoise_Normal}
  NamedRerouteDeclaration(None=Multiply(A=@n33.RGB, B=Constant3Vector() {constant=<rgba 0.015 0.015 1 0>})) {name=Normals, nodeColor=<rgba 1 0 0.6797 1>}
  NamedRerouteDeclaration(None=Add(A=@n17.unclamped_uvs, B=@n31)) {name=unclamped, nodeColor=<rgba 0 0.1328 1 1>}
  NamedRerouteDeclaration(None=@n17.UnPixelatedUVNoClamp) {name=unclampedunpixel, nodeColor=<rgba 1 0.8438 0 1>}
  NamedRerouteDeclaration(None=@n12) {name=emissive, nodeColor=<rgba 1 0.75 0 1>}
  n34 = VectorParameter() {parameterName=BaseColor, defaultValue=<rgba 0.125 0.125 0.125 0>}
  NamedRerouteDeclaration(None=Multiply(A=Multiply(A=@n34.RGB, B=@n1.A), B=Round(None=@n1.RGB))) {name=base, nodeColor=<rgba 1 0 0.3516 1>}
  NamedRerouteUsage()
  NamedRerouteUsage()
  NamedRerouteUsage()
 ]
}
